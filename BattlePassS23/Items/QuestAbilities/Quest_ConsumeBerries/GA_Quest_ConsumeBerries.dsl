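; Passive gameplay ability for a battle pass quest that counts berries the owning player consumes.
; On activation, after calling the parent, it waits on the owning pawn for the ConsumedBerry gameplay event.
; Each received event sends a complex custom stat event with count 1 and the additional source tags to the player's quest manager.
; The failed passive activation handler sets up the same listener and report, without the parent call.
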(event Ability|EventFailedToActivatePassiveAbility
  (params (ActorInfo :type Gameplay%20Ability%20Actor%20Info%20Structure))
  (Ability|Async|WaitGameplayEventToActor (Variables|GAVars|Data|GetOwningPlayerPawn) "(TagName=\"Athena.event.gameplay.ConsumedBerry\")"
    (:EventReceived
      (Quest|SendComplexCustomStatEvent (Variables|GAVars|Data|GetOwningPlayerQuestManager) 0 (Variables|Default|GetAdditional%20Source%20Tags) "(GameplayTags=)" 1))))

(event Ability|EventActivateAbility
  (|CallParentFunction)
  (Ability|Async|WaitGameplayEventToActor (Variables|GAVars|Data|GetOwningPlayerPawn) "(TagName=\"Athena.event.gameplay.ConsumedBerry\")"
    (:EventReceived
      (Quest|SendComplexCustomStatEvent (Variables|GAVars|Data|GetOwningPlayerQuestManager) 0 (Variables|Default|GetAdditional%20Source%20Tags) "(GameplayTags=)" 1))))
